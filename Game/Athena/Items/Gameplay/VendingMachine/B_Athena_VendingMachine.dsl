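; Athena vending machine Blueprint that shows a cycling item with its resource cost and gives audiovisual feedback on vending.
; Cycle updates drive dynamic materials and light with icon, cost and rarity color, picking interact text by matching resource texture paths.
; Timelines animate hit wobble, screen transitions and success/fail flashes, with local-only force feedback.
; Hotfixable settings are cached once playlist data is ready; authority counts spawned pickups and destroys the machine at its limit.
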(event InteractEvents|EventOnForcedCycle
  ;[1 1 1] Set bool for enabling wobble if machine was hit
  (Variables|Default|SetWasHit true)
  (Utilities|FlowControl|RetriggerableDelay 0.5)
  (Variables|Default|SetWasHit false))

(event EventBeginPlay
  (bind _vendingmachine_ambient (Variables|Default|GetVendingMachine_Ambient))
  (bind _sm_vendingmachinequest (Variables|Default|GetSM_VendingMachineQuest))
  ;[1 1 1] BEGIN PLAY
  (Utilities|FlowControl|Sequence
    (:then_0
      ;[1 1 1] Set up audio
      (Audio|Components|Audio|SetSound _vendingmachine_ambient (Variables|Default|GetAmbient_SoundCue))
      (Audio|Components|Audio|Play _vendingmachine_ambient 0.0)
      ;[1 1 1] Set up Monitor Mats
      (bind _returnvalue (Rendering|Material|CreateDynamicMaterialInstance _sm_vendingmachinequest 1 "/Game/Athena/Items/Gameplay/VendingMachine/FX/M_VendingMachineMonitor.M_VendingMachineMonitor" "None" :!class /Script/Engine.PrimitiveComponent))
      (Variables|VendingFX|SetMonitorMat _returnvalue)
      (bind _returnvalue_2 (Rendering|Material|CreateDynamicMaterialInstance _sm_vendingmachinequest 0 "/Game/Athena/Items/Gameplay/VendingMachine/FX/M_VendingMachine.M_VendingMachine" "None" :!class /Script/Engine.PrimitiveComponent))
      (Variables|VendingFX|SetMetalMat _returnvalue_2))
    (:then_1
      ;[1 1 1] Wait for the playlist data to be ready before using hotfixed values.
      (Class|FortAsyncActionCurrentPlaylistReady|PlaylistDataReadyAsync
        (:OnReady
          (bind _output_get (Variables|Default|SetItemsAreFreeCached (== (ScalableFloat|GetValueatLevel (Variables|Default|GetItemsAreFree)) 1.0)))
          (Rendering|Material|SetScalarParameterValue (Variables|VendingFX|GetMonitorMat) "ShowResource" (Math|Conversions|ToFloat(Boolean) (not _output_get)) :!class /Script/Engine.MaterialInstanceDynamic)
          (Variables|Default|SetMaxItemsToSpawnCached (Math|Float|Truncate (ScalableFloat|GetValueatLevel (Variables|Default|GetMaxItemsToSpawn))))
          ;;[1 1 1] Set to Drop on Ground or not
          (bind _outxy (Fortnite|EvaluateCurveTableRow (Variables|Default|GetTossOnGroundSetting)))
          (Variables|LootDrop|SetbTossOnGround (== _outxy 1.0))
          (Networking|SwitchHasAuthority
            (:Authority
              (EventDispatchers|BindEventtoPickupSpawned (EventDispatchers|CreateEvent :!event Custom|PickupSpawned_Bind)))))))))

(event InteractEvents|EventUpdateCollectorsActiveItem
  (params (StartingSeedValue :type Integer) (InputItem :type Fort%20World%20Item%20Definition%20Object%20Reference) (InputTexture :type Texture%20Object%20Reference) (CurrentGoal :type Integer) (CurrentCollectedAmount :type Integer) (CurrentCaptureCount :type Integer) (OutputTexture :type Texture%20Object%20Reference))
  (bind (_value _curve) (Utilities|Struct|BreakScalableFloat (Variables|InteractMethods|GetActiveItemRotationTime)))
  (bind (_rarity _returnvalue) (InteractData|GetOverrideRarity))
  (bind _washit (Variables|Default|GetWasHit))
  (bind _materialtype (Variables|Default|GetMaterialType))
  (bind _monitormat (Variables|VendingFX|GetMonitorMat))
  (bind _monitormat_2 (Variables|VendingFX|GetMonitorMat))
  (bind _monitormat_3 (Variables|VendingFX|GetMonitorMat))
  (bind _monitormat_4 (Variables|VendingFX|GetMonitorMat))
  ;[1 1 1] Update Screen contents
  (Variables|Default|SetMaterialType InputTexture)
  (Variables|Default|SetCostAmount CurrentGoal)
  (Variables|Default|SetCurrentCaptureCount (+ StartingSeedValue CurrentCaptureCount))
  (Rendering|Material|SetTextureParameterValue _monitormat_3 "ResourceIcon" InputTexture)
  (Rendering|Material|SetScalarParameterValue _monitormat_3 "TimerStartTime" (Utilities|Time|GetTimeSeconds) :!class /Script/Engine.MaterialInstanceDynamic)
  (bind _outxy (Fortnite|EvaluateCurveTableRow _curve))
  (Rendering|Material|SetScalarParameterValue _monitormat_3 "TimerDuration" _outxy :!class /Script/Engine.MaterialInstanceDynamic)
  (if (Utilities|IsValid OutputTexture :!class /Script/Engine.KismetSystemLibrary)
    (Rendering|Material|SetTextureParameterValue _monitormat_2 "WeaponIcon" OutputTexture)
    (else
      (Rendering|Material|SetTextureParameterValue _monitormat_2 "WeaponIcon" "/Game/UI/Fort_HUD/MissionWidget/FailedCheckedBox.FailedCheckedBox")))
  (Rendering|Material|SetScalarParameterValue _monitormat_2 "CostNumber" (Math|Conversions|ToFloat(Integer) CurrentGoal) :!class /Script/Engine.MaterialInstanceDynamic)
  ;[1 1 1] Set Material Required Text for Interact
  (bind _equals (CallFunction|CompareTexturetoSoftTexture _materialtype "/Game/UI/Foundation/Textures/Icons/Items/T-Icon-Logs.T-Icon-Logs"))
  (if _equals
    (Variables|Strings|SetVend%20String (Variables|Strings|GetWood%20String))
    (Rendering|Material|SetVectorParameterValue _monitormat "CircleScales" "(R=1.350000,G=0.600000,B=0.600000,A=1.000000)" :!class /Script/Engine.MaterialInstanceDynamic)
    (else
      (bind _equals_2 (CallFunction|CompareTexturetoSoftTexture _materialtype "/Game/UI/Foundation/Textures/Icons/Items/T-Icon-IBeam.T-Icon-IBeam"))
      (if _equals_2
        (Variables|Strings|SetVend%20String (Variables|Strings|GetMetal%20String))
        (Rendering|Material|SetVectorParameterValue _monitormat "CircleScales" "(R=0.600000,G=0.600000,B=1.350000,A=1.000000)" :!class /Script/Engine.MaterialInstanceDynamic)
        (else
          (Variables|Strings|SetVend%20String (Variables|Strings|GetStone%20String))
          (Rendering|Material|SetVectorParameterValue _monitormat "CircleScales" "(R=0.600000,G=1.350000,B=0.600000,A=1.000000)" :!class /Script/Engine.MaterialInstanceDynamic)))))
  ;[1 1 1] Set Color of Vending Machine to proper Tier color
  (switch int (select _returnvalue (Math|Conversions|ToInteger(Byte) _rarity) (Variables|Default|GetCurrentCaptureCount))
    (:Default
      (Rendering|Material|SetVectorParameterValue (Variables|VendingFX|GetMonitorMat) "ScreenColor" "(R=1.000000,G=0.002428,B=0.002428,A=1.000000)" :!class /Script/Engine.MaterialInstanceDynamic))
    (:0
      (CallFunction|SetRarityColors "(R=0.407240,G=0.407240,B=0.407240,A=1.000000)"))
    (:1
      (CallFunction|SetRarityColors "(R=0.147027,G=0.412543,B=0.007499,A=1.000000)"))
    (:2
      (CallFunction|SetRarityColors "(R=0.059511,G=0.337164,B=0.693872,A=1.000000)"))
    (:3
      (CallFunction|SetRarityColors "(R=0.323143,G=0.122139,B=0.533276,A=1.000000)"))
    (:4
      (CallFunction|SetRarityColors "(R=1.000000,G=0.479320,B=0.000000,A=1.000000)")))
  ;[1 1 1] Transition Sound and FX
  (Audio|PlaySoundatLocation (Variables|VendingFX|GetCycleSound) (Transformation|GetActorLocation))
  (bind (_direction _wobble _flash _slide) (Timeline|SwitchHitch PlayFromStart)
    (:Update
      (Transformation|SetRelativeRotation (Variables|Default|GetSM_VendingMachineQuest) (Math|Rotator|MakeRotator (Math|Float|SelectFloat _wobble 0.0 _washit)) false false)
      (Rendering|Material|SetScalarParameterValue _monitormat_4 "Distort" (Math|Float|SelectFloat _flash 0.0 _washit) :!class /Script/Engine.MaterialInstanceDynamic)
      (Rendering|Material|SetScalarParameterValue _monitormat_4 "Slide" _slide :!class /Script/Engine.MaterialInstanceDynamic))))

(event Interaction|EventBlueprintOnBeginInteract
  (params (InteractingPawn :type Fort%20Pawn%20Object%20Reference) (InteractionBeingAttempted :type EInteractionBeingAttempted%20Enum :default "FirstInteraction") (InteractionType :type TInteractionType%20Enum :default "IT_NoInteraction"))
  ;[1 1 1] Search feedback
  (Audio|PlaySoundatLocation (Variables|VendingFX|GetSearchSound) (Transformation|GetActorLocation))
  (Variables|Default|SetCycleSoundVolume 0.2)
  (Utilities|FlowControl|Delay 1.1)
  (Variables|Default|SetCycleSoundVolume 1.0))

(event Interaction|EventBlueprintOnInteract
  (params (InteractingPawn :type Fort%20Pawn%20Object%20Reference) (InteractionBeingAttempted :type EInteractionBeingAttempted%20Enum :default "FirstInteraction"))
  ;[1 1 1] Vending finished feedback
  (Variables|Default|SetInteractingPawn InteractingPawn)
  (if (Variables|InteractMethods|GetbCurrentInteractionSuccess)
    (CallFunction|PlayVendFX)
    (else
      (CallFunction|PlayVendFailFX))))

(event Custom|PlayVendFX
  (bind _monitormat (Variables|VendingFX|GetMonitorMat))
  (bind _interactingpawn (Variables|Default|GetInteractingPawn))
  ;[1 1 1] Vending finished feedback
  (Audio|PlaySoundatLocation (Variables|Default|GetVend%20Succeeded%20Sound) (Transformation|GetActorLocation))
  (Components|Activation|SetActive (Variables|Default|GetP_Loot_Effect_Vending) true true)
  (Variables|Default|SetVendSuccess true)
  (bind (_direction _wobbletime _flash) (Timeline|VendWobble PlayFromStart)
    (:Update
      (Transformation|SetRelativeRotation (Variables|Default|GetSM_VendingMachineQuest) (Math|Rotator|Lerp(Rotator) "0, 0, 0" "0.500000,0.000000,0.000000" _wobbletime true) false false)
      (Rendering|Material|SetScalarParameterValue (Variables|VendingFX|GetMonitorMat) (select (Variables|Default|GetVendSuccess) "Success" "Fail") _flash :!class /Script/Engine.MaterialInstanceDynamic))
    (:Finished
      (Rendering|Material|SetScalarParameterValue _monitormat "Success" 0.0 :!class /Script/Engine.MaterialInstanceDynamic) ; Reset both params
      (Rendering|Material|SetScalarParameterValue _monitormat "Fail" 0.0 :!class /Script/Engine.MaterialInstanceDynamic))
    (:Start%20FF%20Time
      (if (Variables|Default|GetVendSuccess)
        (Utilities|IsValid _interactingpawn
          (:Is%20Valid
            (if (Pawn|IsLocallyControlled _interactingpawn)
              (Components|Activation|Activate (Variables|Default|GetForceFeedbackSuccess) false))))))))

(event Custom|PlayVendFailFX
  (bind _monitormat (Variables|VendingFX|GetMonitorMat))
  (bind _interactingpawn (Variables|Default|GetInteractingPawn))
  ;[1 1 1] Vending finished feedback
  (Audio|PlaySoundatLocation (Variables|VendingFX|GetVendFailedSound) (Transformation|GetActorLocation))
  (Variables|Default|SetVendSuccess false)
  (if (Pawn|IsLocallyControlled (Variables|Default|GetInteractingPawn))
    (Components|Activation|Activate (Variables|Default|GetForceFeedbackFail) false))
  (Timeline|VendWobble PlayFromStart))

(event Custom|PickupSpawned_Bind
  ;[1 1 1] CUSTOM EVENT -> PickupSpawned_Bind
  (bind _result (Math|Integer|IncrementInt (Variables|Default|GetSpawnedItemCount))
    (:%20%20
      (if (and (>= _result (Variables|Default|GetMaxItemsToSpawnCached)) (> (Variables|Default|GetMaxItemsToSpawnCached) 0))
        (CallFunction|DoVendDeath)
        (Actor|DestroyActor)))))

(event Custom|DoVendDeath
  ;[1 1 1] CUSTOM EVENT -> DoVendDeath
  ;;[0.1742 1 0.1335] Do death FX.
  ;; Adjust lifespan time to match.
  (Audio|PlaySoundatLocation (Variables|Default|GetDestroyedSound) (Transformation|GetActorLocation))
  (Components|Activation|SetActive (Variables|Default|GetP_Destroyed_Effect_Vending) true true)
  (Rendering|SetVisibility (Variables|Default|GetSM_VendingMachineQuest) false false))

(fn BlueprintGetInteractionString
  (params (InteractingPawn :type Fort%20Pawn%20Object%20Reference) (InteractionBeingAttempted :type EInteractionBeingAttempted%20Enum :default "FirstInteraction"))
  (return (Utilities|Text|ToText(String) (select (> (Variables|Default|GetCostAmount) 0) (Utilities|String|Append (Utilities|String|ToString(Text) (Variables|Strings|GetVend%20String)) " (" (Utilities|String|ToString(Integer) (Variables|Default|GetCostAmount)) ")") (Utilities|String|ToString(Text) (Variables|Strings|GetFree%20String))))))

(fn CompareTextureToSoftTexture
  (params (Texture :type Texture%20Object%20Reference) (Soft%20Texture :type Texture%20Soft%20Object%20Reference))
  (return (Utilities|String|EqualExactly(String) (Utilities|BreakSoftObjectPath (Utilities|GetSoftObjectPath Texture)) (Utilities|BreakSoftObjectPath (Utilities|ToSoftObjectPath Soft%20Texture)))))

(fn GetOverrideRarity
  (bind (_rarity _returnvalue) (|CallParentFunction))
  (Utilities|FlowControl|Sequence
    (:then_0
      (LocalVariables|SetNewRarity _rarity)
      (bind _output_get (LocalVariables|SetbHasOverride _returnvalue))
      (if _output_get
        _
        (elif (Variables|InteractMethods|InstanceOverride|GetbUseInstanceLootValueOverrides)
          (LocalVariables|SetNewRarity (Variables|InteractMethods|InstanceOverride|GetOverrideVendingMachineRarity))
          (LocalVariables|SetbHasOverride true))))
    (:then_1
      (return (LocalVariables|GetNewRarity) (LocalVariables|GetbHasOverride)))))

(fn SetRarityColors
  (params (Color :type Linear%20Color%20Structure))
  (Rendering|Material|SetVectorParameterValue (Variables|VendingFX|GetMonitorMat) "ScreenColor" Color :!class /Script/Engine.MaterialInstanceDynamic)
  (Rendering|Components|Light|SetLightColor (Variables|Default|GetSpotLight) Color false)
  (Rendering|Material|SetVectorParameterValue (Variables|VendingFX|GetMetalMat) "PaintColor" Color :!class /Script/Engine.MaterialInstanceDynamic))

(fn ConstructionScript)
