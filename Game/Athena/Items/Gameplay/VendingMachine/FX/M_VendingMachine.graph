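# Opaque lit material for the Athena vending machine, with a parameterized paint tint.
# A color mask blends the diffuse texture toward a version multiplied by PaintColor, which is first lerped slightly by a constant.
# The green channel of the retail snack machine prop's mask texture drives Metallic, Specular and Roughness alike, alongside that prop's normal map.
# The assembled material attributes pass through MF_CreativeMoveTool, apparently for Creative move-tool feedback.
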Material /Game/Athena/Items/Gameplay/VendingMachine/FX/M_VendingMachine
graph {
 properties {bUseMaterialAttributes=true, bUsedWithNanite=true, pixelDepthOffsetMode=PDOM_Legacy, usedShadingModels=Default Lit}
 nodes [
  n1 = TextureSampleParameter2D() {parameterName=Diffuse, texture=&/T_VendingMachine_D}
  n2 = LinearInterpolate(A=VectorParameter() {parameterName=PaintColor, defaultValue=<rgba 1 0.4667 0.1469 0>}.RGB) {constAlpha=0.1}
  n3 = TextureSample() {samplerType=SAMPLERTYPE_Masks, texture=/Game/Environments/Props/Retail/Textures/Prop_Snack_Vending_Machine_01_M}
  n4 = TextureSampleParameter2D() {parameterName=Normals, samplerType=SAMPLERTYPE_Normal, texture=/Game/Environments/Props/Retail/Textures/Prop_Snack_Vending_Machine_01_N}
  n5 = SetMaterialAttributes!(
   MaterialAttributes
   Base Color=LinearInterpolate(
    A=@n1.RGB
    B=Multiply(A=@n1.RGB, B=@n2)
    Alpha=TextureSample() {texture=&/T_VendingMachineColorMask}.R
   )
   Metallic=@n3.G
   Specular=@n3.G
   Roughness=@n3.G
   Normal=@n4.RGB
  ) {attributeSetTypes [
    #1
    #2
    #3
    #4
    #5
   ]}
  MaterialFunctionCall(MA In=@n5) {materialFunction=&MF_CreativeMoveTool}
 ]
}
